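# Physics asset for the F_MED_Gimmick female medium body cosmetic, covering seven dynamic collar pieces (A–G) and the upper body around them.
# Capsules on three spine bones and both upper arms form the body volumes, linked by moderate angular limits.
# Each collar chain has a root capsule held nearly rigid to spine_03 and a box tip with limited swing and very little twist relative to that root.
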PhysicsAsset /BRCosmetics/Characters/Player/Female/Medium/Bodies/F_MED_Gimmick/Meshes/F_MED_Gimmick_Physics
bodies {
 dyn_BodycollarA_01 [{shapeType=Capsule, length=5.25, radius=2, shapeName=dyn_BodycollarA_01_capsule, center=<1.5061 -0.0001 0>}]
 dyn_BodycollarA_02 [{shapeType=Box, extentX=5.5, extentY=2.5, extentZ=7.5, shapeName=dyn_BodycollarA_02_box, center=<2.2731 0 0>, rotation=<rot 0 4.2376 0>}]
 dyn_BodycollarB_01 [{shapeType=Capsule, length=5.25, radius=2, shapeName=dyn_BodycollarB_01_capsule, center=<1.5061 -0.0001 0>}]
 dyn_BodycollarB_02 [{shapeType=Box, extentX=5.5, extentY=2.5, extentZ=7.5, shapeName=dyn_BodycollarB_02_box, center=<2.2731 0 0>, rotation=<rot 0 4.2379 0>}]
 dyn_BodycollarC_01 [{shapeType=Capsule, length=5.25, radius=2, shapeName=dyn_BodycollarC_01_capsule, center=<1.5061 -0.0001 0>}]
 dyn_BodycollarC_02 [{shapeType=Box, extentX=5.5, extentY=3, extentZ=7.5, shapeName=dyn_BodycollarC_02_box, center=<2.2731 0 0>, rotation=<rot 0 4.2339 0>}]
 dyn_BodycollarD_01 [{shapeType=Capsule, length=5.25, radius=2, shapeName=dyn_BodycollarD_01_capsule, center=<1.5061 -0.0001 0>}]
 dyn_BodycollarD_02 [{shapeType=Box, extentX=5.5, extentY=3, extentZ=7.5, shapeName=dyn_BodycollarD_02_box, center=<2.2731 0 0>, rotation=<rot 0 4.2373 0>}]
 dyn_BodycollarE_01 [{shapeType=Capsule, length=5.25, radius=2, shapeName=dyn_BodycollarE_01_capsule, center=<1.5061 -0.0001 0>}]
 dyn_BodycollarE_02 [{shapeType=Box, extentX=5.5, extentY=2.5, extentZ=7.5, shapeName=dyn_BodycollarE_02_box, center=<2.2731 0 0>, rotation=<rot 0 4.2388 0>}]
 dyn_BodycollarF_01 [{shapeType=Capsule, length=5.25, radius=2, shapeName=dyn_BodycollarF_01_capsule, center=<1.5061 -0.0001 0>}]
 dyn_BodycollarF_02 [{shapeType=Box, extentX=5.5, extentY=2.5, extentZ=7.5, shapeName=dyn_BodycollarF_02_box, center=<2.2731 0 0>, rotation=<rot 0 4.2379 0>}]
 dyn_BodycollarG_01 [{shapeType=Capsule, length=5.25, radius=2, shapeName=dyn_BodycollarG_01_capsule, center=<1.5061 -0.0001 0>}]
 dyn_BodycollarG_02 [{shapeType=Box, extentX=5.5, extentY=3, extentZ=7.5, shapeName=dyn_BodycollarG_02_box, center=<2.2731 0 0>, rotation=<rot 0 4.2365 0>}]
 spine_01 [{shapeType=Capsule, length=3.0479, radius=7.4385, center=<9.9193 1.7444 -0.0967>, rotation=<rot 0.0938 -89.9998 0.0145>}]
 spine_03 [{shapeType=Capsule, length=2.75, radius=9.5545, center=<5.8135 3.0032 -0.1847>, rotation=<rot -9.3302 -90 1.5647>}]
 spine_05 [{shapeType=Capsule, length=15.7679, radius=8.31, center=<4.0137 2.2883 0.0307>, rotation=<rot -0.077 -90 0.3793>}]
 upperarm_l [{shapeType=Capsule, length=14.5198, radius=7.4546, center=<12.322 -1.817 1.5544>, rotation=<rot 0 -91.6633 -85.9726>}]
 upperarm_r [{shapeType=Capsule, length=14.2544, radius=7.4463, center=<-12.2759 1.8158 -1.3495>, rotation=<rot 0 87.9991 -89.6381>}]
}
constraints [
 {bone1Name=spine_03, bone2Name=spine_01, swing1LimitDegrees=45, swing2LimitDegrees=45, twistLimitDegrees=45}
 {bone1Name=spine_05, bone2Name=spine_03, swing1LimitDegrees=45, swing2LimitDegrees=45, twistLimitDegrees=45}
 {bone1Name=upperarm_l, bone2Name=spine_05, swing1LimitDegrees=45, swing2LimitDegrees=45, twistLimitDegrees=45}
 {bone1Name=upperarm_r, bone2Name=spine_05, swing1LimitDegrees=45, swing2LimitDegrees=45, twistLimitDegrees=45}
 {bone1Name=dyn_BodycollarA_01, bone2Name=spine_03, swing1LimitDegrees=0.1, swing2LimitDegrees=0.1, twistLimitDegrees=0.1}
 {bone1Name=dyn_BodycollarA_02, bone2Name=dyn_BodycollarA_01, swing1LimitDegrees=23, swing2LimitDegrees=10, twistLimitDegrees=2}
 {bone1Name=dyn_BodycollarB_01, bone2Name=spine_03, swing1LimitDegrees=0.1, swing2LimitDegrees=0.1, twistLimitDegrees=0.1}
 {bone1Name=dyn_BodycollarB_02, bone2Name=dyn_BodycollarB_01, swing1LimitDegrees=23, swing2LimitDegrees=10, twistLimitDegrees=2}
 {bone1Name=dyn_BodycollarC_01, bone2Name=spine_03, swing1LimitDegrees=0.1, swing2LimitDegrees=0.1, twistLimitDegrees=0.1}
 {bone1Name=dyn_BodycollarC_02, bone2Name=dyn_BodycollarC_01, swing1LimitDegrees=23, swing2LimitDegrees=10, twistLimitDegrees=2}
 {bone1Name=dyn_BodycollarD_01, bone2Name=spine_03, swing1LimitDegrees=0.1, swing2LimitDegrees=0.1, twistLimitDegrees=0.1}
 {bone1Name=dyn_BodycollarD_02, bone2Name=dyn_BodycollarD_01, swing1LimitDegrees=23, swing2LimitDegrees=10, twistLimitDegrees=2}
 {bone1Name=dyn_BodycollarE_01, bone2Name=spine_03, swing1LimitDegrees=0.1, swing2LimitDegrees=0.1, twistLimitDegrees=0.1}
 {bone1Name=dyn_BodycollarE_02, bone2Name=dyn_BodycollarE_01, swing1LimitDegrees=23, swing2LimitDegrees=10, twistLimitDegrees=2}
 {bone1Name=dyn_BodycollarF_01, bone2Name=spine_03, swing1LimitDegrees=0.1, swing2LimitDegrees=0.1, twistLimitDegrees=0.1}
 {bone1Name=dyn_BodycollarF_02, bone2Name=dyn_BodycollarF_01, swing1LimitDegrees=23, swing2LimitDegrees=10, twistLimitDegrees=2}
 {bone1Name=dyn_BodycollarG_01, bone2Name=spine_03, swing1LimitDegrees=0.1, swing2LimitDegrees=0.1, twistLimitDegrees=0.1}
 {bone1Name=dyn_BodycollarG_02, bone2Name=dyn_BodycollarG_01, swing1LimitDegrees=23, swing2LimitDegrees=10, twistLimitDegrees=2}
]
