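# Physics asset for the F_MED_Gimmick face accessory mesh, setting up hair physics for this female medium player body.
# Box and capsule bodies form six jointed hair chains (side and back) rooted at the head, alongside head, neck, spine and upper-arm capsules.
# Angular limits range from near-rigid on the left side to looser swing down the back.
# Cross-constraints tie lower back strands together, right-side hair to back-right, and left-side tips to the neck.
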PhysicsAsset /BRCosmetics/Characters/Player/Female/Medium/Bodies/F_MED_Gimmick/Meshes/Parts/F_MED_Gimmick_FaceAcc_Physics
bodies {
 dyn_HairA_l_01 [{shapeType=Box, extentX=7.5, extentY=3, extentZ=3, shapeName=dyn_HairA_l_01_box, center=<-1.9918 0 0>, rotation=<rot 0 0 17.5357>}]
 dyn_HairA_l_02 [{shapeType=Box, extentX=4.5, extentY=3, extentZ=3.5, shapeName=dyn_HairA_l_02_box, center=<-1.228 0 -0.3407>, rotation=<rot 9.6145 0 17.5475>}]
 dyn_HairA_l_base [{
   shapeType=Capsule
   length=1.25
   radius=1
   shapeName=dyn_HairA_l_base_capsule
   center=<-0.525 0 0.0041>
   rotation=<rot -89.5537 -179.9995 179.9994>
  }]
 dyn_Hair_BackL_01 [{
   shapeType=Box
   extentX=3.5
   extentY=7.3139
   extentZ=10
   shapeName=dyn_Hair_BackL_01_box
   center=<0.1386 -0.2366 0>
   rotation=<rot 0 -34.7148 0>
  }]
 dyn_Hair_BackL_02 [{
   shapeType=Box
   extentX=2.8082
   extentY=7.9771
   extentZ=10
   shapeName=dyn_Hair_BackL_02_box
   center=<0.1734 -1.5849 -0.0627>
   rotation=<rot -30.1407 -4.5473 1.5714>
  }]
 dyn_Hair_BackL_03 [{
   shapeType=Box
   extentX=3.331
   extentY=16.55
   extentZ=9
   shapeName=dyn_Hair_BackL_03_box
   center=<0.1119 -2.315 -1.1533>
   rotation=<rot -22.5848 -6.1998 -7.4269>
  }]
 dyn_Hair_BackL_04 [{
   shapeType=Capsule
   length=4.8604
   radius=6.25
   shapeName=dyn_Hair_BackL_04_capsule
   center=<2.4737 -2.6939 -0.3414>
   rotation=<rot -0.0001 0 80.545>
  }]
 dyn_Hair_BackL_05 [{
   shapeType=Capsule
   length=5.75
   radius=6.25
   shapeName=dyn_Hair_BackL_05_capsule
   center=<0.8953 -2.4721 -0.0049>
   rotation=<rot -0.0001 0 91.3248>
  }]
 dyn_Hair_BackL_06 [{
   shapeType=Capsule
   length=6.4214
   radius=5.5786
   shapeName=dyn_Hair_BackL_06_capsule
   center=<0.8952 -2.7602 0.0018>
   rotation=<rot -0.0001 0 91.3248>
  }]
 dyn_Hair_BackR_01 [{shapeType=Box, extentX=3.5, extentY=7.5, extentZ=10, shapeName=dyn_Hair_BackR_01_box, center=<0.1328 -0.0352 0>, rotation=<rot 0 -56.0851 0>}]
 dyn_Hair_BackR_02 [{
   shapeType=Box
   extentX=3.3719
   extentY=7.6463
   extentZ=10
   shapeName=dyn_Hair_BackR_02_box
   center=<0.2335 -0.3951 0.0751>
   rotation=<rot 35.6018 -0.7074 -0.4928>
  }]
 dyn_Hair_BackR_03 [{
   shapeType=Box
   extentX=3.6035
   extentY=12.5
   extentZ=9
   shapeName=dyn_Hair_BackR_03_box
   center=<0.5519 -3.9549 1.135>
   rotation=<rot 33.8102 5.0793 15.7417>
  }]
 dyn_Hair_BackR_04 [{
   shapeType=Capsule
   length=4
   radius=5.25
   shapeName=dyn_Hair_BackR_04_capsule
   center=<2.7393 -2.2937 0.1556>
   rotation=<rot 0.7791 9.6235 104.7471>
  }]
 dyn_Hair_BackR_05 [{
   shapeType=Capsule
   length=3.2117
   radius=6.5383
   shapeName=dyn_Hair_BackR_05_capsule
   center=<2.7957 -3.2778 0.6225>
   rotation=<rot -0.1791 7.7653 91.3126>
  }]
 dyn_Hair_BackR_06 [{
   shapeType=Capsule
   length=4
   radius=5.75
   shapeName=dyn_Hair_BackR_06_capsule
   center=<0.8953 -3.2918 0.014>
   rotation=<rot -0.0001 0 91.3248>
  }]
 dyn_Hair_Back_01 [{shapeType=Box, extentX=3.5, extentY=7.5, extentZ=10, shapeName=dyn_Hair_Back_01_box, center=<0.3523 -0.3428 0>, rotation=<rot 0 -64.2819 0>}]
 dyn_Hair_Back_02 [{
   shapeType=Box
   extentX=3.9306
   extentY=9.3752
   extentZ=10
   shapeName=dyn_Hair_Back_02_box
   center=<0.5485 -1.4889 0>
   rotation=<rot 0 -2.8372 0>
  }]
 dyn_Hair_Back_03 [{shapeType=Box, extentX=2, extentY=8, extentZ=10, shapeName=dyn_Hair_Back_03_box, center=<0 -2.5214 0>}]
 dyn_Hair_Back_04 [{
   shapeType=Capsule
   length=5.75
   radius=5.5
   shapeName=dyn_Hair_Back_04_capsule
   center=<1.6903 -2.2744 -0.0183>
   rotation=<rot -0.0001 0 91.3248>
  }]
 dyn_Hair_Back_05 [{
   shapeType=Capsule
   length=5.75
   radius=6.25
   shapeName=dyn_Hair_Back_05_capsule
   center=<0.8954 -2.8087 -0.0041>
   rotation=<rot -0.0001 0 91.3248>
  }]
 dyn_Hair_Back_06 [{
   shapeType=Capsule
   length=7.2886
   radius=6.25
   shapeName=dyn_Hair_Back_06_capsule
   center=<0.9955 -1.6073 -0.0319>
   rotation=<rot 0.0989 -4.2827 91.321>
  }]
 dyn_Hair_l_01 [{
   shapeType=Capsule
   length=5
   radius=2.5
   shapeName=dyn_Hair_l_01_capsule
   center=<-2.2288 -0.4527 -1.1531>
   rotation=<rot 8.6642 -82.9486 91.0676>
  }]
 dyn_Hair_l_02 [{
   shapeType=Box
   extentX=7
   extentY=7.5
   extentZ=3.5
   shapeName=dyn_Hair_l_02_box
   center=<-3.1106 0 -1.2063>
   rotation=<rot -3.3245 -0.0001 0>
  }]
 dyn_Hair_l_03 [{shapeType=Box, extentX=7, extentY=6, extentZ=4.5, shapeName=dyn_Hair_l_03_box, center=<-1.3482 0 -1.5168>, rotation=<rot -12.1114 -0.0001 0>}]
 dyn_Hair_l_04 [{
   shapeType=Box
   extentX=6
   extentY=4
   extentZ=4.3848
   shapeName=dyn_Hair_l_04_box
   center=<-0.7311 0.6081 1.0777>
   rotation=<rot -10.7536 -34.5239 5.6123>
  }]
 dyn_Hair_l_05 [{
   shapeType=Capsule
   length=2.0038
   radius=2.5544
   shapeName=dyn_Hair_l_05_capsule
   center=<-0.9037 0.8202 1.5018>
   rotation=<rot 12.0458 32.5469 71.8903>
  }]
 dyn_Hair_l_base [{shapeType=Capsule, length=1.5, radius=2.25, shapeName=dyn_Hair_l_base_capsule, center=<0.2906 0 -0.5424>, rotation=<rot -61.8217 0 0>}]
 dyn_Hair_r_01 [{shapeType=Capsule, length=4.25, radius=2.25, shapeName=dyn_Hair_r_01_capsule, rotation=<rot -82.5516 0.0003 -0.0002>}]
 dyn_Hair_r_02 [{
   shapeType=Capsule
   length=2.8226
   radius=3.6774
   shapeName=dyn_Hair_r_02_capsule
   center=<-0.316 0 2.416>
   rotation=<rot -82.549 0.0004 -0.0003>
  }]
 dyn_Hair_r_03 [{shapeType=Capsule, length=0.5, radius=3.4188, shapeName=dyn_Hair_r_03_capsule, center=<-0.2519 0 1.929>, rotation=<rot -82.5609 0 0>}]
 dyn_Hair_r_04 [{shapeType=Capsule, length=3.8435, radius=3.4065, shapeName=dyn_Hair_r_04_capsule, center=<-1.0618 0 0.8346>, rotation=<rot -38.1693 0 0>}]
 dyn_Hair_r_05 [{
   shapeType=Capsule
   length=4.5681
   radius=2.682
   shapeName=dyn_Hair_r_05_capsule
   center=<-1.4289 0.1312 0.8551>
   rotation=<rot -30.7906 -5.2447 26.5797>
  }]
 dyn_Hair_r_base [{shapeType=Capsule, length=1.75, radius=1.75, shapeName=dyn_Hair_r_base_capsule}]
 head [{shapeType=Capsule, length=7.04, radius=6.96, shapeName=head_capsule, center=<2.0202 0.1733 -0.2592>, rotation=<rot 90 -0.02 -0.0202>}]
 neck_02 [{shapeType=Capsule, length=6, radius=5.25, shapeName=neck_02_capsule, rotation=<rot 89.1755 179.9955 179.9955>}]
 spine_04 [{shapeType=Capsule, length=31.25, radius=8.75, shapeName=spine_04_capsule, center=<-0.9724 -0.1996 0>}]
 spine_05 [
  {shapeType=Capsule, length=21.8356, radius=8.3126, shapeName=spine_05_capsule, center=<5.3981 0.6208 0.0001>}
  {
   shapeType=Capsule
   length=7.6702
   radius=4.8298
   shapeName=spine_05_capsule
   center=<6.7069 5.8968 6.2505>
   rotation=<rot -0.782 17.4244 -87.5093>
  }
 ]
 upperarm_l [{
   shapeType=Capsule
   length=23.4084
   radius=6.0916
   shapeName=upperarm_l_capsule
   center=<14.7078 -0.0002 1.2361>
   rotation=<rot 84.4941 179.9998 179.9998>
  }]
 upperarm_r [{
   shapeType=Capsule
   length=22.3726
   radius=6.6274
   shapeName=upperarm_r_capsule
   center=<-12.4095 -1.1701 0.6239>
   rotation=<rot 84.4955 179.9998 179.9998>
  }]
}
constraints [
 {bone1Name=dyn_HairA_l_base, bone2Name=head, swing1LimitDegrees=30, swing2LimitDegrees=20, twistLimitDegrees=10}
 {bone1Name=dyn_HairA_l_01, bone2Name=dyn_HairA_l_base, swing1LimitDegrees=10, swing2LimitDegrees=5, twistLimitDegrees=0.1}
 {bone1Name=dyn_HairA_l_02, bone2Name=dyn_HairA_l_01, swing1LimitDegrees=30, swing2LimitDegrees=20, twistLimitDegrees=0.1}
 {bone1Name=dyn_Hair_l_base, bone2Name=head, swing1LimitDegrees=1, swing2LimitDegrees=1, twistLimitDegrees=1}
 {bone1Name=dyn_Hair_l_01, bone2Name=dyn_Hair_l_base, swing1LimitDegrees=1, swing2LimitDegrees=1, twistLimitDegrees=1}
 {bone1Name=dyn_Hair_l_02, bone2Name=dyn_Hair_l_01, swing1LimitDegrees=0.1, swing2LimitDegrees=0.1, twistLimitDegrees=0.1}
 {bone1Name=dyn_Hair_l_03, bone2Name=dyn_Hair_l_02, swing1LimitDegrees=0.1, swing2LimitDegrees=0.1, twistLimitDegrees=0.1}
 {bone1Name=dyn_Hair_l_04, bone2Name=dyn_Hair_l_03, swing1LimitDegrees=0.1, swing2LimitDegrees=0.1, twistLimitDegrees=0.1}
 {bone1Name=dyn_Hair_l_05, bone2Name=dyn_Hair_l_04, swing1LimitDegrees=0.1, swing2LimitDegrees=0.1, twistLimitDegrees=0.1}
 {
  bone1Name=dyn_Hair_r_base
  bone2Name=head
  swing1LimitDegrees=30
  swing1Motion=Locked
  swing2LimitDegrees=20
  swing2Motion=Locked
  twistLimitDegrees=10
  twistMotion=Locked
 }
 {
  bone1Name=dyn_Hair_r_01
  bone2Name=dyn_Hair_r_base
  swing1LimitDegrees=0.1
  swing1Motion=Locked
  swing2LimitDegrees=0.1
  swing2Motion=Locked
  twistLimitDegrees=0.1
  twistMotion=Locked
 }
 {bone1Name=dyn_Hair_r_02, bone2Name=dyn_Hair_r_01, swing1LimitDegrees=5, swing2LimitDegrees=15, twistLimitDegrees=0.1}
 {bone1Name=dyn_Hair_r_03, bone2Name=dyn_Hair_r_02, swing1LimitDegrees=5, swing2LimitDegrees=15, twistLimitDegrees=0.1}
 {bone1Name=dyn_Hair_r_04, bone2Name=dyn_Hair_r_03, swing1LimitDegrees=10, swing2LimitDegrees=10, twistLimitDegrees=0.1}
 {bone1Name=dyn_Hair_r_05, bone2Name=dyn_Hair_r_04, swing1LimitDegrees=5, swing2LimitDegrees=15, twistLimitDegrees=0.1}
 {bone1Name=dyn_Hair_Back_01, bone2Name=head, swing1LimitDegrees=0.1, swing2LimitDegrees=0.1, twistLimitDegrees=0.1}
 {bone1Name=dyn_Hair_Back_02, bone2Name=dyn_Hair_Back_01, swing1LimitDegrees=15, swing2LimitDegrees=5, twistLimitDegrees=5, twistMotion=Locked}
 {bone1Name=dyn_Hair_Back_03, bone2Name=dyn_Hair_Back_02, swing1LimitDegrees=30, swing2LimitDegrees=10, twistLimitDegrees=0.1}
 {bone1Name=dyn_Hair_Back_04, bone2Name=dyn_Hair_Back_03, swing1LimitDegrees=30, swing2LimitDegrees=10, twistLimitDegrees=1}
 {bone1Name=dyn_Hair_Back_05, bone2Name=dyn_Hair_Back_04, swing1LimitDegrees=30, swing2LimitDegrees=10, twistLimitDegrees=1}
 {bone1Name=dyn_Hair_Back_06, bone2Name=dyn_Hair_Back_05, swing1LimitDegrees=30, swing2LimitDegrees=10, twistLimitDegrees=1}
 {bone1Name=dyn_Hair_BackL_01, bone2Name=head, swing1LimitDegrees=0.1, swing2LimitDegrees=0.1, twistLimitDegrees=0.1}
 {bone1Name=dyn_Hair_BackL_02, bone2Name=dyn_Hair_BackL_01, swing1LimitDegrees=15, swing2LimitDegrees=5, twistLimitDegrees=5}
 {bone1Name=dyn_Hair_BackL_03, bone2Name=dyn_Hair_BackL_02, swing1LimitDegrees=30, swing2LimitDegrees=10, twistLimitDegrees=10}
 {bone1Name=dyn_Hair_BackL_04, bone2Name=dyn_Hair_BackL_03, swing1LimitDegrees=30, swing2LimitDegrees=10, twistLimitDegrees=10}
 {bone1Name=dyn_Hair_BackL_05, bone2Name=dyn_Hair_BackL_04, swing1LimitDegrees=30, swing2LimitDegrees=10, twistLimitDegrees=0.1}
 {bone1Name=dyn_Hair_BackL_06, bone2Name=dyn_Hair_BackL_05, swing1LimitDegrees=30, swing2LimitDegrees=10, twistLimitDegrees=0.1}
 {bone1Name=dyn_Hair_BackR_01, bone2Name=head, swing1LimitDegrees=0.1, swing2LimitDegrees=0.1, twistLimitDegrees=0.1}
 {bone1Name=dyn_Hair_BackR_02, bone2Name=dyn_Hair_BackR_01, swing1LimitDegrees=15, swing2LimitDegrees=5, twistLimitDegrees=5}
 {bone1Name=dyn_Hair_BackR_03, bone2Name=dyn_Hair_BackR_02, swing1LimitDegrees=30, swing2LimitDegrees=10, twistLimitDegrees=10}
 {bone1Name=dyn_Hair_BackR_04, bone2Name=dyn_Hair_BackR_03, swing1LimitDegrees=30, swing2LimitDegrees=10, twistLimitDegrees=10}
 {bone1Name=dyn_Hair_BackR_05, bone2Name=dyn_Hair_BackR_04, swing1LimitDegrees=30, swing2LimitDegrees=10, twistLimitDegrees=0.1}
 {bone1Name=dyn_Hair_BackR_06, bone2Name=dyn_Hair_BackR_05, swing1LimitDegrees=30, swing2LimitDegrees=10, twistLimitDegrees=0.1}
 {
  bone1Name=dyn_Hair_BackL_04
  bone2Name=dyn_Hair_Back_04
  swing1LimitDegrees=10
  swing1Motion=Free
  swing2LimitDegrees=10
  swing2Motion=Free
  twistLimitDegrees=10
 }
 {
  bone1Name=dyn_Hair_BackR_04
  bone2Name=dyn_Hair_Back_04
  swing1LimitDegrees=10
  swing1Motion=Free
  swing2LimitDegrees=10
  swing2Motion=Free
  twistLimitDegrees=10
 }
 {
  bone1Name=dyn_Hair_BackL_05
  bone2Name=dyn_Hair_Back_05
  swing1LimitDegrees=10
  swing1Motion=Free
  swing2LimitDegrees=10
  swing2Motion=Free
  twistLimitDegrees=10
 }
 {
  bone1Name=dyn_Hair_BackR_05
  bone2Name=dyn_Hair_Back_05
  swing1LimitDegrees=10
  swing1Motion=Free
  swing2LimitDegrees=10
  swing2Motion=Free
  twistLimitDegrees=10
 }
 {
  bone1Name=dyn_Hair_BackL_06
  bone2Name=dyn_Hair_Back_06
  swing1LimitDegrees=10
  swing1Motion=Free
  swing2LimitDegrees=10
  swing2Motion=Free
  twistLimitDegrees=10
 }
 {
  bone1Name=dyn_Hair_BackR_06
  bone2Name=dyn_Hair_Back_06
  swing1LimitDegrees=10
  swing1Motion=Free
  swing2LimitDegrees=10
  swing2Motion=Free
  twistLimitDegrees=10
 }
 {
  bone1Name=neck_02
  bone2Name=dyn_Hair_l_05
  swing1LimitDegrees=45
  swing1Motion=Locked
  swing2LimitDegrees=45
  swing2Motion=Locked
  twistLimitDegrees=45
  twistMotion=Locked
 }
 {
  bone1Name=neck_02
  bone2Name=dyn_Hair_l_04
  swing1LimitDegrees=45
  swing1Motion=Locked
  swing2LimitDegrees=45
  swing2Motion=Locked
  twistLimitDegrees=45
  twistMotion=Locked
 }
 {bone1Name=dyn_Hair_r_03, bone2Name=dyn_Hair_BackR_03, swing1LimitDegrees=45, swing2LimitDegrees=45, twistLimitDegrees=45}
 {bone1Name=dyn_Hair_r_05, bone2Name=dyn_Hair_BackR_04, swing1LimitDegrees=45, swing2LimitDegrees=45, twistLimitDegrees=45}
]
